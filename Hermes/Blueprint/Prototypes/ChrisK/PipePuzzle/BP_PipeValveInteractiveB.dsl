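; Interactable pipe-puzzle valve prototype that turns on interaction, notifies its pipe valve manager, then winds back after a delay.
; One Turn timeline lerps the mesh rotation forward and in reverse, with play rate recalculated per direction.
; Delays time the reversal and ResetValve call, while interaction is gated by valve state and puzzle completion.
; SwitchHasAuthority limits turn audio and steam to remote instances; the solved event disables interaction and plays a stinger.
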(event EventBeginPlay
  ;[1 1 1] Init
  (bind _output_get (Variables|Default|SetInitialRelativeRotation (Variables|/Script/Engine.SceneComponent|GetRelativeRotation (Variables|Components|GetValveMesh))))
  (Variables|Default|SetEndRelativeRotation (Math|Rotator|CombineRotators _output_get "-179.000000,0.000000,0.000000"))
  (Variables|Default|SetValveActive false))

(event Interaction|EventBlueprintOnInteract
  (params (InteractingPawn :type Fort%20Pawn%20Object%20Reference) (InteractionBeingAttempted :type EInteractionBeingAttempted%20Enum :default "FirstInteraction"))
  ;[1 1 1] Turn the valve
  (Utilities|FlowControl|DoOnce false
    (:Completed
      (Variables|Default|SetCanPlayerInteract false)
      (Variables|Default|SetValveActive true)
      (Class|BPPipeValveManager|ValveTurned (Variables|Default|GetPipeValveManagerRef))
      (Utilities|FlowControl|Sequence
        (:then_0
          (Variables|Default|SetCurrentTimeLineSpeedMultiplier (/ (Variables|Default|GetDefaultTimelineSpeed) (Variables|Default|GetTimeForInitialValveTurn)))
          (Components|Timeline|SetPlayRate (Variables|Components|GetTurn) (Variables|Default|GetCurrentTimeLineSpeedMultiplier))
          (Variables|Default|SetIsReversing false)
          (CallFunction|ClientFX false)
          (bind (_direction _rotatevalve) (Timeline|Turn PlayFromStart)
            (:Update
              (Transformation|SetRelativeRotation (Variables|Components|GetValveMesh) (Math|Rotator|Lerp(Rotator) (Variables|Default|GetInitialRelativeRotation) (Variables|Default|GetEndRelativeRotation) _rotatevalve true) false false))
            (:Finished
              (if (Variables|Default|GetIsReversing)
                (CallFunction|ClientFXReverseFinished)))))
        (:then_1
          (Utilities|FlowControl|Delay (* (Variables|Default|GetDefaultTimelineSpeed) (Variables|Default|GetTimeForInitialValveTurn)))
          (Utilities|FlowControl|Delay (Variables|Default|GetDelayAfterOpeningBeforeClosing))
          (Variables|Default|SetCurrentTimeLineSpeedMultiplier (/ (Variables|Default|GetDefaultTimelineSpeed) (Variables|Default|GetValveDurationToReset)))
          (Components|Timeline|SetPlayRate (Variables|Components|GetTurn) (Variables|Default|GetCurrentTimeLineSpeedMultiplier))
          (Utilities|FlowControl|Sequence
            (:then_0
              (Variables|Default|SetIsReversing true)
              (CallFunction|ClientFX true)
              (Timeline|Turn ReverseFromEnd))
            (:then_1
              (Utilities|FlowControl|Delay (* (Variables|Default|GetDefaultTimelineSpeed) (Variables|Default|GetValveDurationToReset)))
              (CallFunction|ResetValve))))))))

(event Custom|ResetValve
  ;[1 1 1] Turn the valve
  (Variables|Default|SetCanPlayerInteract true)
  (Variables|Default|SetValveActive false)
  (if (not (Variables|Default|GetPuzzleIsSolved))
    (Utilities|FlowControl|DoOnce false
      (:Completed
        (Variables|Default|SetCanPlayerInteract false)
        (Variables|Default|SetValveActive true)
        (Class|BPPipeValveManager|ValveTurned (Variables|Default|GetPipeValveManagerRef))
        (Utilities|FlowControl|Sequence
          (:then_0
            (Variables|Default|SetCurrentTimeLineSpeedMultiplier (/ (Variables|Default|GetDefaultTimelineSpeed) (Variables|Default|GetTimeForInitialValveTurn)))
            (Components|Timeline|SetPlayRate (Variables|Components|GetTurn) (Variables|Default|GetCurrentTimeLineSpeedMultiplier))
            (Variables|Default|SetIsReversing false)
            (CallFunction|ClientFX false)
            (Timeline|Turn PlayFromStart))
          (:then_1
            (Utilities|FlowControl|Delay (* (Variables|Default|GetDefaultTimelineSpeed) (Variables|Default|GetTimeForInitialValveTurn)))
            (Utilities|FlowControl|Delay (Variables|Default|GetDelayAfterOpeningBeforeClosing))
            (Variables|Default|SetCurrentTimeLineSpeedMultiplier (/ (Variables|Default|GetDefaultTimelineSpeed) (Variables|Default|GetValveDurationToReset)))
            (Components|Timeline|SetPlayRate (Variables|Components|GetTurn) (Variables|Default|GetCurrentTimeLineSpeedMultiplier))
            (Utilities|FlowControl|Sequence
              (:then_0
                (Variables|Default|SetIsReversing true)
                (CallFunction|ClientFX true)
                (Timeline|Turn ReverseFromEnd))
              (:then_1
                (Utilities|FlowControl|Delay (* (Variables|Default|GetDefaultTimelineSpeed) (Variables|Default|GetValveDurationToReset)))
                (CallFunction|ResetValve)))))))))

(event Custom|ClientFX
  (params (IsReversing :type Boolean))
  ;[1 1 1] Client sounds and FX
  (Networking|SwitchHasAuthority
    (:Remote
      (if IsReversing
        (Audio|Components|Audio|Play (Variables|Components|GetAudio_ReverseOneShot) 0.0)
        (Audio|Components|Audio|Play (Variables|Default|GetAudio_ReversingLoop) 0.0)
        (else
          (Audio|Components|Audio|Play (Variables|Components|GetAudio_TurnActivatedForward) 0.0)))
      (Components|Activation|Activate (Variables|Components|GetNS_Parallel_Manhole_Steam) false)
      (Utilities|FlowControl|Delay (Variables|Default|GetDelayAfterOpeningBeforeClosing))
      (Utilities|FlowControl|Delay (* (Variables|Default|GetDefaultTimelineSpeed) (Variables|Default|GetValveDurationToReset)))
      (Components|Activation|Deactivate (Variables|Components|GetNS_Parallel_Manhole_Steam)))))

(event Custom|ClientFXReverseFinished
  ;[1 1 1] Client sounds and FX
  (Networking|SwitchHasAuthority
    (:Remote
      (Audio|Components|Audio|Play (Variables|Default|GetAudio_ReversingFinished) 0.0)
      (Audio|Components|Audio|FadeOut (Variables|Default|GetAudio_ReversingLoop) 0.33 0.0 "Linear" :!class /Script/Engine.AudioComponent))))

(event Custom|PuzzleSolvedDisableInteract
  ;[1 1 1] Puzzle Solved Disable Interaction
  (Variables|Interaction|SetbAllowInteract false)
  (Variables|Default|SetCanPlayerInteract false)
  (Audio|SpawnSoundatLocation "/Hermes/Audio/Pipe_Puzzle/Valves/Interactables_S34_Pipe_Puzzle_Stinger_Cue.Interactables_S34_Pipe_Puzzle_Stinger_Cue" (+ (Transformation|GetWorldLocation (Variables|Debug|GetStaticMeshComponent)) "0.000000,0.000000,30.000000")))

(fn BlueprintCanInteract
  (params (InteractingPawn :type Fort%20Pawn%20Object%20Reference) (InteractionBeingAttempted :type EInteractionBeingAttempted%20Enum :default "FirstInteraction") (InteractionType :type TInteractionType%20Enum :default "IT_NoInteraction"))
  (if (Variables|Default|GetPuzzleIsSolved)
    (return false)
    (else
      (return (Variables|Default|GetCanPlayerInteract)))))

(fn BlueprintGetInteractionString
  (params (InteractingPawn :type Fort%20Pawn%20Object%20Reference) (InteractionBeingAttempted :type EInteractionBeingAttempted%20Enum :default "FirstInteraction"))
  (return (Variables|Default|GetInteractionString)))

(fn ConstructionScript)
