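; Manager for a pipe valve puzzle that opens a pressure door once all of its valves are active.
; ValveTurned does nothing if the puzzle is already solved; otherwise it checks the active state of every valve in its array.
; Success marks the manager and valves solved, opens the door and disables valve interaction.
; Otherwise the door plays a partial open and a debug "Failed" message prints.
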(event Custom|ValveTurned
  (if (not (Variables|Default|GetPuzzleSolvedAtManager))
    (bind _allactive (CallFunction|CheckValveStates))
    (if _allactive
      (CallFunction|SolvedPuzzle)
      (else
        (CallFunction|PartialDoorAnim)
        (Development|PrintString "Failed" true true "(R=0.000000,G=0.660000,B=1.000000,A=1.000000)" 5.0)))))

(fn CheckValveStates
  (bind _array_element (Utilities|Array|ForEachLoopwithBreak (Variables|Default|GetAllPipeValves))
    (:LoopBody
      (if (not (Variables|/Hermes/Blueprint/Prototypes/ChrisK/PipePuzzle/BP_PipeValveInteractive.BP_PipeValveInteractive_C|GetValveActive _array_element))
        (LocalVariables|SetValveActiveLocal false)))
    (:Completed
      _))
  (return (LocalVariables|GetValveActiveLocal)))

(fn PartialDoorAnim
  (Class|BPPressureDoorInteractive|PartOpenDoors (Variables|Default|GetPressureDoorRef)))

(fn SolvedPuzzle
  (for _array_element (Variables|Default|GetAllPipeValves)
    (Variables|/Hermes/Blueprint/Prototypes/ChrisK/PipePuzzle/BP_PipeValveInteractive.BP_PipeValveInteractive_C|SetPuzzleIsSolved true _array_element))
  (Variables|Default|SetPuzzleSolvedAtManager true)
  (Class|BPPressureDoorInteractive|OpenDoor (Variables|Default|GetPressureDoorRef))
  (for _array_element_2 (Variables|Default|GetAllPipeValves)
    (Utilities|IsValid _array_element_2
      (:Is%20Valid
        (Class|BPPipeValveInteractive|PuzzleSolvedDisableInteract _array_element_2)))))

(fn ConstructionScript)
