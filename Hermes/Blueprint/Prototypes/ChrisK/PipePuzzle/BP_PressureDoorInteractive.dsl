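; Pressure door prototype for a pipe puzzle: two panels that slide apart along world X, with partial and full opening stages.
; PartOpenDoors lerps the panels slightly open on a timeline, flags completion, then replays that motion once via ResetPartial.
; OpenDoor retries every second until the partial stage completes, then disables the RepNotify-driven collision blocker and plays the full slide.
; DoOnce gates guard the stages, and opening sounds play only without authority.
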(event EventBeginPlay
  ; Init
  (Variables|Default|SetCollisionBlockerEnabled true)
  (bind _output_get (Variables|Default|SetInitRightWorldLoc (Transformation|GetWorldLocation (Variables|Default|GetDoorRight))))
  (Variables|Default|SetInitRightEndWorldLoc (+ _output_get "150.000000,0.000000,0.000000"))
  (bind _output_get_2 (Variables|Default|SetInitLeftWorldLoc (Transformation|GetWorldLocation (Variables|Default|GetDoorLeft))))
  (Variables|Default|SetInitLeftEndWorldLoc (+ _output_get_2 "-150.000000,0.000000,0.000000")))

(event Custom|RetryOpen
  ; Real Opening Anim
  (if (and (Variables|Default|GetDoorOpened) (Variables|Default|GetPartlyOpenedCompleted))
    (Utilities|FlowControl|DoOnce false
      (:Completed
        (Variables|Default|SetCollisionBlockerEnabled false)
        (CallFunction|RealDoorOpeningClientFX)
        (bind (_direction _movecrate) (Timeline|AnimagteCrate PlayFromStart)
          (:Update
            (Transformation|SetWorldLocation (Variables|Default|GetDoorRight) (Math|Vector|Lerp(Vector) (Variables|Default|GetInitRightWorldLoc) (Variables|Default|GetInitRightEndWorldLoc) _movecrate) false false)
            (Transformation|SetWorldLocation (Variables|Default|GetDoorLeft) (Math|Vector|Lerp(Vector) (Variables|Default|GetInitLeftWorldLoc) (Variables|Default|GetInitLeftEndWorldLoc) _movecrate) false false)))))
    (else
      (Utilities|FlowControl|Delay 1.0)
      (CallFunction|RetryOpen))))

(event Custom|OpenDoor
  ; Real Opening Anim
  (Variables|Default|SetDoorOpened true)
  (if (and (Variables|Default|GetDoorOpened) (Variables|Default|GetPartlyOpenedCompleted))
    (Utilities|FlowControl|DoOnce false
      (:Completed
        (Variables|Default|SetCollisionBlockerEnabled false)
        (CallFunction|RealDoorOpeningClientFX)
        (Timeline|AnimagteCrate PlayFromStart)))
    (else
      (Utilities|FlowControl|Delay 1.0)
      (CallFunction|RetryOpen))))

(event Custom|RealDoorOpeningClientFX
  ; Real Opening Anim
  (Networking|SwitchHasAuthority
    (:Remote
      (Audio|Components|Audio|Play (Variables|Default|GetFullOpenAudio) 0.0))))

(event Custom|PartOpenDoors
  ; PartialDoorOpeningAnim
  (Variables|Default|SetPartlyOpenedCompleted false)
  (Utilities|FlowControl|DoOnce false
    (:Completed
      (Utilities|FlowControl|Delay 1.0)
      (CallFunction|PartialOpenClientFX)
      (bind (_direction _partlymoveanim) (Timeline|Timeline PlayFromStart)
        (:Update
          (Transformation|SetWorldLocation (Variables|Default|GetDoorRight) (Math|Vector|Lerp(Vector) (Variables|Default|GetInitRightWorldLoc) (+ (Variables|Default|GetInitRightEndWorldLoc) "-140.000000,0.000000,0.000000") _partlymoveanim) false false)
          (Transformation|SetWorldLocation (Variables|Default|GetDoorLeft) (Math|Vector|Lerp(Vector) (Variables|Default|GetInitLeftWorldLoc) (+ (Variables|Default|GetInitLeftEndWorldLoc) "140.000000,0.000000,0.000000") _partlymoveanim) false false))
        (:Finished
          (Utilities|FlowControl|Delay 0.25)
          (Variables|Default|SetPartlyOpenedCompleted true)
          (CallFunction|ResetPartial))))))

(event Custom|ResetPartial
  ; PartialDoorOpeningAnim
  (Utilities|FlowControl|DoOnce false
    (:Completed
      (Utilities|FlowControl|Delay 1.0)
      (CallFunction|PartialOpenClientFX)
      (Timeline|Timeline PlayFromStart))))

(event Custom|PartialOpenClientFX
  ; PartialDoorOpeningAnim
  (Networking|SwitchHasAuthority
    (:Remote
      (Audio|Components|Audio|Play (Variables|Default|GetPartialOpenAudio) 0.0))))

(fn OnRep_CollisionBlockerEnabled
  (if (Variables|Default|GetCollisionBlockerEnabled)
    (Collision|SetCollisionEnabled (Variables|Default|GetDoorBlockerCollision) "QueryAndPhysics")
    (else
      (Collision|SetCollisionEnabled (Variables|Default|GetDoorBlockerCollision) "NoCollision"))))

(fn ConstructionScript)
